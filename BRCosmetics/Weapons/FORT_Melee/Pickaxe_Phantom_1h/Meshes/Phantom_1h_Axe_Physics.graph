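# Physics asset for the Phantom one-handed axe pickaxe cosmetic, defining bodies and joints for its dangling wrap and tail bones.
# Capsules cover the base and upper wrap segments, small capsules sit on their `_base` bones, and four flat boxes form the tail.
# Constraints link everything into one chain from base through upper wrap to tail.
# Free-rotating joints decouple the three sections, while limited joints within them keep the tail tighter than the wrap.
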PhysicsAsset /BRCosmetics/Weapons/FORT_Melee/Pickaxe_Phantom_1h/Meshes/Phantom_1h_Axe_Physics
bodies {
 dyn_wrap_base [{shapeType=Capsule, length=5.5, radius=5, center=<2 0 0>, rotation=<rot 90 0 0>}]
 dyn_wrap_base_base [{shapeType=Capsule, length=1, radius=1}]
 dyn_wrap_tail_a [{shapeType=Box, extentX=8, extentY=1.5, extentZ=5.5, center=<4 0 0>}]
 dyn_wrap_tail_a_base [{shapeType=Capsule, length=1, radius=1}]
 dyn_wrap_tail_b [{shapeType=Box, extentX=9.5, extentY=1.5, extentZ=5.5, center=<4.8532 0 0>}]
 dyn_wrap_tail_c [{shapeType=Box, extentX=9, extentY=1.5, extentZ=5.5, center=<4.9148 0 0>}]
 dyn_wrap_tail_d [{shapeType=Box, extentX=8, extentY=1.5, extentZ=5.5, center=<4 0 0>}]
 dyn_wrap_up [{shapeType=Capsule, length=5.5, radius=5, center=<2 0 0>, rotation=<rot 90 0 0>}]
 dyn_wrap_up_base [{shapeType=Capsule, length=1, radius=1}]
}
constraints [
 {
  bone1Name=dyn_wrap_up_base
  bone2Name=dyn_wrap_base
  swing1LimitDegrees=45
  swing1Motion=Free
  swing2LimitDegrees=45
  swing2Motion=Free
  twistLimitDegrees=45
  twistMotion=Free
 }
 {
  bone1Name=dyn_wrap_tail_a_base
  bone2Name=dyn_wrap_up
  swing1LimitDegrees=45
  swing1Motion=Free
  swing2LimitDegrees=45
  swing2Motion=Free
  twistLimitDegrees=45
  twistMotion=Free
 }
 {bone1Name=dyn_wrap_base, bone2Name=dyn_wrap_base_base, swing1LimitDegrees=45, swing2LimitDegrees=45, twistLimitDegrees=45}
 {bone1Name=dyn_wrap_up, bone2Name=dyn_wrap_up_base, swing1LimitDegrees=45, swing2LimitDegrees=45, twistLimitDegrees=45}
 {bone1Name=dyn_wrap_tail_a, bone2Name=dyn_wrap_tail_a_base, swing1LimitDegrees=10, swing2LimitDegrees=10, twistLimitDegrees=5}
 {bone1Name=dyn_wrap_tail_b, bone2Name=dyn_wrap_tail_a, swing1LimitDegrees=10, swing2LimitDegrees=10, twistLimitDegrees=5}
 {bone1Name=dyn_wrap_tail_c, bone2Name=dyn_wrap_tail_b, swing1LimitDegrees=10, swing2LimitDegrees=10, twistLimitDegrees=5}
 {bone1Name=dyn_wrap_tail_d, bone2Name=dyn_wrap_tail_c, swing1LimitDegrees=10, swing2LimitDegrees=10, twistLimitDegrees=5}
]
