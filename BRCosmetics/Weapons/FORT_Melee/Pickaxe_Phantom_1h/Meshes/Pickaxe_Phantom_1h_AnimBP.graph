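# Animation Blueprint for the Phantom one-handed pickaxe that makes its wrap and tail bones drift and flutter procedurally.
# Each update, noise_FNC turns game time into sine-shaped Perlin noise for translation and rotation variables.
# These additively drive wrap bases before a RigidBody simulation and four tail bones after it; an AnimDynamics node sits unconnected.
# Initialization gives the WeaponMeshOffhand component a noise phase offset, desynchronizing it from the other hand.
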AnimBlueprint /BRCosmetics/Weapons/FORT_Melee/Pickaxe_Phantom_1h/Meshes/Pickaxe_Phantom_1h_AnimBP
target_skeleton=&/Phantom_1h_Skeleton
graphs [
 AnimGraph {nodes [
   Anim Dynamics - Bone: dyn_wrap_base(Alpha=1, ExternalForce="0,0,0") {
    bGravityOverrideInSimSpace=true
    gravityScale=0.35
    simulationSpace=RootRelative
    alphaBoolBlend {blendOption=Linear}
    boundBone {boneName=dyn_wrap_base}
    physicsBodyDefinitions [{
      sphereCollisionRadius=0
      boundBone {boneName=dyn_wrap_base}
      boxExtents=<5 20 20>
      constraintSetup {angularLimitsMax=<0 5 5>, angularLimitsMin=<0 -5 -5>, angularTarget=<1 0 0>}
      localJointOffset=<-10 0 0>
     }]
   }
   n1 = Get tranNoiseA() -> (tranNoiseA:Vector)
   n2 = "Transform (Modify) Bone - Bone: dyn_wrap_base_base"(ComponentPose=Local To Component(LocalPose=Input Pose()), Alpha=1, Translation=@n1, Rotation="0,0,0", Scale="1,1,1") {
    rotationMode=BMM_Additive
    rotationSpace=BCS_BoneSpace
    scaleMode=BMM_Additive
    scaleSpace=BCS_BoneSpace
    translationMode=BMM_Additive
    translationSpace=BCS_BoneSpace
    alphaBoolBlend {blendOption=Linear}
    boneToModify {boneName=dyn_wrap_base_base}
   }
   n3 = "Transform (Modify) Bone - Bone: dyn_wrap_up_base"(ComponentPose=@n2, Alpha=1, Translation=@n1, Rotation="0,0,0", Scale="1,1,1") {
    rotationMode=BMM_Additive
    rotationSpace=BCS_BoneSpace
    scaleMode=BMM_Additive
    scaleSpace=BCS_BoneSpace
    translationMode=BMM_Additive
    translationSpace=BCS_BoneSpace
    alphaBoolBlend {blendOption=Linear}
    boneToModify {boneName=dyn_wrap_up_base}
   }
   n4 = RigidBody(ComponentPose=@n3) {
    bClampLinearTranslationLimitToRefPose=true
    bOverrideWorldGravity=true
    overridePhysicsAsset=&/Phantom_1h_Axe_Physics
    simulationSpace=BaseBoneSpace
    alphaBoolBlend {blendOption=Linear}
    baseBoneRef {boneName=b_AR_Root}
    componentAppliedLinearAccClamp=<1000 1000 1000>
    componentLinearAccScale=<0.2 0.2 0.2>
    componentLinearVelScale=<0.2 0.2 0.2>
   }
   n5 = "Transform (Modify) Bone - Bone: dyn_wrap_tail_a"(ComponentPose=@n4, Alpha=1, Translation="0,0,0", Rotation=Get rotationNoiseA() -> (rotationNoiseA:Rotator), Scale="1,1,1") {
    rotationMode=BMM_Additive
    rotationSpace=BCS_BoneSpace
    scaleMode=BMM_Additive
    scaleSpace=BCS_BoneSpace
    translationMode=BMM_Additive
    translationSpace=BCS_BoneSpace
    alphaBoolBlend {blendOption=Linear}
    boneToModify {boneName=dyn_wrap_tail_a}
   }
   n6 = Get tranNoiseB() -> (tranNoiseB:Vector)
   n7 = "Transform (Modify) Bone - Bone: dyn_wrap_tail_b"(ComponentPose=@n5, Alpha=1, Translation=@n6, Rotation=Get rotationNoiseB() -> (rotationNoiseB:Rotator), Scale="1,1,1") {
    rotationMode=BMM_Additive
    rotationSpace=BCS_BoneSpace
    scaleMode=BMM_Additive
    scaleSpace=BCS_BoneSpace
    translationMode=BMM_Additive
    translationSpace=BCS_BoneSpace
    alphaBoolBlend {blendOption=Linear}
    boneToModify {boneName=dyn_wrap_tail_b}
   }
   n8 = "Transform (Modify) Bone - Bone: dyn_wrap_tail_c"(ComponentPose=@n7, Alpha=1, Translation=@n6, Rotation=Get rotationNoiseC() -> (rotationNoiseC:Rotator), Scale="1,1,1") {
    rotationMode=BMM_Additive
    rotationSpace=BCS_BoneSpace
    scaleMode=BMM_Additive
    scaleSpace=BCS_BoneSpace
    translationMode=BMM_Additive
    translationSpace=BCS_BoneSpace
    alphaBoolBlend {blendOption=Linear}
    boneToModify {boneName=dyn_wrap_tail_c}
   }
   n9 = "Transform (Modify) Bone - Bone: dyn_wrap_tail_d"(ComponentPose=@n8, Alpha=1, Translation=@n6, Rotation=Get rotationNoiseD() -> (rotationNoiseD:Rotator), Scale="1,1,1") {
    rotationMode=BMM_Additive
    rotationSpace=BCS_BoneSpace
    scaleMode=BMM_Additive
    scaleSpace=BCS_BoneSpace
    translationMode=BMM_Additive
    translationSpace=BCS_BoneSpace
    alphaBoolBlend {blendOption=Linear}
    boneToModify {boneName=dyn_wrap_tail_d}
   }
   Output Pose(Result=Component To Local(ComponentPose=@n9)) {layerGroup=None}
  ]}
 EventGraph {nodes [
   Event BlueprintUpdateAnimation()
   > n1 = Sequence!(execute) -> (then_0, then_1, then_2:Exec, then_3:Exec)
   n2 = Get offset_weapon() -> (offset_weapon:double)
   > n3 = noise_FNC!(execute, self, frequency:double=0.35, amp:double=3, offset:double=@n2) -> (then, output:double)
   n4 = noise_FNC(execute=@n1.then_1, frequency=0.2, amp=1.5, offset=@n2)
   Set tranNoiseA(execute=@n3.then, tranNoiseA:Vector=MakeVector(X=@n3, Y=@n4, Z=@n4)) -> (Output_Get:Vector)
   n5 = Sequence(execute=@n1.then_2)
   n6 = Get offset_weapon()
   > n7 = noise_FNC(frequency=0.5, amp=5, offset=@n6)
   n8 = noise_FNC(execute=@n5.then_2, frequency=0.5, amp=5, offset=@n6)
   n9 = noise_FNC(execute=@n5.then_1, frequency=0.5, amp=5, offset=@n6)
   Set rotationNoiseA(execute=@n7.then, rotationNoiseA:Rotator=MakeRotator(Roll=@n7, Pitch=@n9, Yaw=@n8)) -> (Output_Get:Rotator)
   n10 = Sequence(execute=@n5.then_3)
   n11 = Get offset_weapon()
   > n12 = noise_FNC(frequency=0.3, amp=5, offset=@n11)
   n13 = noise_FNC(execute=@n10.then_1, frequency=0.35, amp=30, offset=@n11)
   n14 = noise_FNC(execute=@n10.then_2, frequency=0.45, amp=30, offset=@n11)
   Set rotationNoiseB(execute=@n12.then, rotationNoiseB:Rotator=MakeRotator(Roll=@n12, Pitch=@n13, Yaw=@n14)) -> (Output_Get:Rotator)
   n15 = Sequence(execute=@n10.then_3)
   n16 = Get offset_weapon()
   > n17 = noise_FNC(frequency=0.35, amp=30, offset=@n16)
   n18 = noise_FNC(execute=@n15.then_1, frequency=0.35, amp=30, offset=@n16)
   n19 = noise_FNC(execute=@n15.then_2, frequency=0.4, amp=30, offset=@n16)
   Set rotationNoiseC(execute=@n17.then, rotationNoiseC:Rotator=MakeRotator(Roll=@n17, Pitch=@n18, Yaw=@n19)) -> (Output_Get:Rotator)
   n20 = Sequence(execute=@n15.then_3)
   n21 = Get offset_weapon()
   > n22 = noise_FNC(frequency=0.2, amp=30, offset=@n21)
   n23 = noise_FNC(execute=@n20.then_2, frequency=0.2, amp=-50, offset=@n21)
   n24 = noise_FNC(execute=@n20.then_1, frequency=0.2, amp=-40, offset=@n21)
   Set rotationNoiseD(execute=@n22.then, rotationNoiseD:Rotator=MakeRotator(Roll=@n22, Pitch=@n24, Yaw=@n23)) -> (Output_Get:Rotator)
   Event BlueprintInitializeAnimation()
   > n25 = Set Mesh_Name(Mesh_Name:Skeletal Mesh Component Object Reference=GetPartSkeletalMeshComponent()) -> (Output_Get:Skeletal Mesh Component Object Reference)
   > n26 = Branch(Condition="Equal Exactly (String)"(A="To String (Object)"!(InObj:Object Reference=@n25) -> (ReturnValue:String), B=WeaponMeshOffhand))
   > Set offset_weapon(offset_weapon:double=0.7) -> (Output_Get:double)
   Set offset_weapon(execute=@n26.else, offset_weapon=0)
   TryGetPawnOwner()
   Set tranNoiseB(execute=@n4.then, tranNoiseB:Vector=MakeVector(Y=@n4, Z=@n4)) -> (Output_Get:Vector)
  ]}
 noise_FNC {nodes [
   n1 = noise_FNC!() -> (then, frequency:double=~, amp:double=~, offset:double=~) {is=title K2Node_FunctionEntry}
   n2 = float + float(A=float * float(A=@n1, B=GetGameTimeInSeconds()), B=@n1.offset) {is=title K2Node_CommutativeAssociativeBinaryOperator}
   n3 = float * float(A="Sin (Radians)"(A=PerlinNoise1D!(Value:float=@n2) -> (ReturnValue:float)), B=@n1.amp)
   > Return Node!(execute, output:double=@n3)
   PerlinNoise1D(Value=@n3)
  ]}
]
variables [
 Mesh_Name {type=SkeletalMeshComponent}
 offset_weapon {type=number}
 rotationNoiseA {type=Rotator}
 rotationNoiseB {type=Rotator}
 rotationNoiseC {type=Rotator}
 rotationNoiseD {type=Rotator}
 tranNoiseA {type=Vector}
 tranNoiseB {type=Vector}
]
